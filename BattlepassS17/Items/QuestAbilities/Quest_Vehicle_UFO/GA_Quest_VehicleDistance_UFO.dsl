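; Gameplay ability for a Season 17 Battle Pass quest, named for tracking distance traveled in a UFO vehicle.
; It overrides only the event-response check, reading gameplay tags from the avatar actor through the Fort vehicle interface.
; The parent check is called, but the returned result comes solely from comparing the Tag Case 0 variable with those vehicle tags.
; No distance-tracking logic appears in this graph.
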(fn ShouldAbilityRespondToEvent
  (params (ActorInfo :type Gameplay%20Ability%20Actor%20Info%20Structure) (Payload :type Gameplay%20Event%20Data%20Structure))
  (bind _returnvalue (Class|FortVehicleInterface|GetGameplayTags(Message) (Ability|GetAvatarActorfromActorInfo)))
  (|CallParentFunction)
  (return (GameplayTags|TagContainer|CompareTagContainertoOtherTagContainers (Variables|Default|GetTag%20Case%200) _returnvalue)))
